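# Additive, unlit, two-sided material for the DJ Synth skydiving contrail mesh and ribbon, drawing a glowing equalizer grid.
# UVs are quantized into bands whose lit cells follow panned beat and rhythm textures, scaled by a MasterBeatMult dynamic parameter.
# Banded gradients drive hue shifting, while spherical masks and depth fades shape opacity.
# World position offset adds noise, gravity and optional rhythm scaling, with only the latter kept on mobile and low quality.
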Material /BRCosmetics/Athena/Effects/Materials/Skydiving/P_Contrail_DJ_Synth/ModeA/M_DJ_Synth_Equalizer_MeshRibbon_A
graph {
 properties {
  bUsedWithBeamTrails=true
  bUsedWithMeshParticles=true
  blendMode=BLEND_Additive
  pixelDepthOffsetMode=PDOM_Legacy
  shadingModel=MSM_Unlit
  twoSided=true
  usedShadingModels=Unlit
 }
 nodes [
  n1 = Frac(None=Multiply(A=Time(), B=ScalarParameter() {defaultValue=0.1, group=Color, parameterName=Chroma Time Dialation}))
  n2 = TextureCoordinate()
  n3 = ScalarParameter() {defaultValue=16, group=EQ Shape, parameterName=V Band Samples}
  n4 = Multiply(A=StaticComponentMaskParameter(None=@n2) {defaultR=true, group=Mask, parameterName=V_Mask}, B=@n3) {constB=10}
  n5 = Sine(None=Divide(A=Ceil(None=@n4), B=@n3)) {period=2}
  n6 = Power(Base=@n5) {constExponent=320}
  n7 = Power(Base=@n5) {constExponent=32}
  n8 = OneMinus(None=StaticComponentMaskParameter(None=TextureCoordinate()) {defaultG=true, group=Mask, parameterName=Mask Component Mask})
  n9 = ScalarParameter() {defaultValue=15, group=Mask, parameterName=U Band Samples}
  n10 = Divide(A=Ceil(None=Multiply(A=@n8, B=@n9) {constB=10}), B=@n9)
  n11 = ScalarParameter() {group=Mask, parameterName=U Band Bias}
  n12 = Multiply(A=@n5, B=Saturate(None=Add(A=@n10, B=@n11)))
  n13 = Add(A=Min(A=@n7, B=Power(Base=@n12)), B=Constant() {r=-0.05})
  n14 = Add(A=@n12, B=@n5)
  n15 = Multiply(A=Add(A=Multiply(A=@n7, B=@n13), B=Subtract(A=@n14) {constB=1.05}), B=@n14)
  n16 = StaticSwitchParameter(True=Saturate(None=Add(A=@n6, B=@n15)), False=Saturate(None=@n15)) {defaultValue=true, parameterName=DJ Ribbon}
  n17 = StaticSwitchParameter(True=Saturate(None=@n16), False=Saturate(None=@n12)) {defaultValue=true, group=Color, parameterName=Alt Chroma Type, sortPriority=-2}
  n18 = OneMinus(None=@n17)
  n19 = Add(
   A=ScalarParameter() {defaultValue=0.5, group=Color, parameterName=Chroma Shift Add, sliderMin=0.001}
   B=Multiply(
    A=ScalarParameter() {defaultValue=0.5, group=Color, parameterName=Chroma Shift Mult, sliderMin=0.001}
    B=StaticSwitchParameter(True=@n18, False=@n17) {group=Color, parameterName=Invert Colors, sortPriority=-1}
   )
  )
  n20 = MaterialFunctionCall(
   Hue Shift Percentage=StaticSwitchParameter(True=Add(A=@n1, B=@n19), False=@n19) {group=Color, parameterName=Chroma Time Shift, sortPriority=-2}
   Texture=VectorParameter() {group=Color, parameterName=ChromaBase, defaultValue=<rgba 0.6 0.6 0.1 1>}.RGB
  ) {materialFunction=&HueShift}
  n21 = ScalarParameter() {defaultValue=0.25, group=Emissive, parameterName=Chroma_Beat_Low}
  n22 = ScalarParameter() {defaultValue=10, group=Emissive, parameterName=Chroma_Beat_High}
  n23 = ScalarParameter() {defaultValue=1, group=EQ Shape, parameterName=EQ_Tiling, sortPriority=1}
  n24 = Constant()
  n25 = ScalarParameter() {defaultValue=15, group=EQ Shape, parameterName=U Band Samples}
  n26 = Multiply(A=ComponentMask(None=TextureCoordinate()) {r=true}, B=@n25)
  n27 = Divide(A=Ceil(None=AppendVector(A=@n24, B=@n26)), B=@n25)
  n28 = ScalarParameter() {group=Beat Texture, parameterName=U Speed, sortPriority=2}
  n29 = ScalarParameter() {group=Beat Texture, parameterName=V Speed, sortPriority=3}
  n30 = TextureSampleParameter2D!(UVs=Panner(Coordinate=Multiply(A=@n23, B=@n27) {constA=0.05, constB=0.05}, Speed=AppendVector(A=@n28, B=@n29)), Level, Apply View MipBias) {
   group=Beat Texture
   mipValueMode=TMVM_MipLevel
   parameterName=Beat Texture Input
   sortPriority=-5
   texture=/Game/Effects/Fort_Effects/BluGlow/ZoomAnimation
  }
  n31 = Ceil(None=Multiply(A=ChannelMaskParameter(None=@n30.RGB) {group=Beat Texture, parameterName=Beat Channel Selector, sortPriority=-4}.R, B=@n3))
  n32 = Multiply(A=TextureCoordinate(), B=Constant2Vector() {r=1})
  n33 = TextureSampleParameter2D(UVs=Panner(Coordinate=Ceil(None=@n32), Speed=ScalarParameter() {defaultValue=0.5, group=Rhythm Texture, parameterName=Rhythm Speed, sortPriority=1})) {
   group=Rhythm Texture
   mipValueMode=TMVM_MipLevel
   parameterName=Rhythm Texture Input
   samplerType=SAMPLERTYPE_LinearColor
   texture=&HighContrastNoise
  }
  n34 = ScalarParameter() {defaultValue=1.2, group=Rhythm Texture, parameterName=Rhythm Mult High, sortPriority=2}
  n35 = ScalarParameter() {defaultValue=0.8, group=Rhythm Texture, parameterName=Rhythm Mult Low, sortPriority=3}
  n36 = LinearInterpolate(
   B=LinearInterpolate(A=Multiply(A=@n33.G, B=@n34), B=Multiply(A=@n33.G, B=@n35), Alpha=@n33.B)
   Alpha=ScalarParameter() {defaultValue=0.8, group=Rhythm Texture, parameterName=rhythm A FADE}
  ) {constA=1}
  n37 = Power(
   Base=Multiply(A=Divide(A=@n31, B=@n3), B=@n36)
   Exp=ScalarParameter() {defaultValue=1, group=Beat Texture, parameterName=Freq Power, sortPriority=4}
  )
  n38 = ScalarParameter() {defaultValue=1, group=Beat Texture, parameterName=Freq Mult, sortPriority=5}
  n39 = ChannelMaskParameter(None=@n30.RGB) {group=Beat Texture, maskChannel=Green, parameterName=Secondary Beat Channel Selector, sortPriority=-3, defaultValue=<rgba 0 1 0 0>}
  n40 = Ceil(None=Multiply(A=@n39.R, B=@n3))
  n41 = LinearInterpolate(
   A=Multiply(A=@n33.R, B=ScalarParameter() {defaultValue=1.2, group=Rhythm Texture, parameterName=Secondary Rhythm Mult High, sortPriority=4})
   B=Multiply(A=@n33.R, B=ScalarParameter() {defaultValue=0.8, group=Rhythm Texture, parameterName=Secondary Rhythm Mult Low, sortPriority=5})
   Alpha=@n33.B
  )
  n42 = ScalarParameter() {defaultValue=0.8, group=Rhythm Texture, parameterName=rhythm B FADE}
  n43 = Power(
   Base=Multiply(A=Divide(A=@n40, B=@n3), B=LinearInterpolate(B=@n41, Alpha=@n42) {constA=1})
   Exp=ScalarParameter() {defaultValue=1, group=Beat Texture, parameterName=Secndary Freq Power, sortPriority=6}
  )
  n44 = ScalarParameter() {defaultValue=1, group=Beat Texture, parameterName=Secondary Freq Mult, sortPriority=7}
  n45 = Saturate(None=Add(A=Multiply(A=@n37, B=@n38), B=Multiply(A=@n43, B=@n44)))
  n46 = Power(Base=@n45, Exp=ScalarParameter() {defaultValue=3, group=Emissive, parameterName=Chroma_Beat_Power}) {constExponent=3}
  n47 = Frac(None=@n4)
  n48 = Multiply(A=OneMinus(None=@n47), B=@n47)
  n49 = DynamicParameter() {defaultValue=<rgba 0 1 0 0>, paramNames [Random, MasterBeatMult, 0, 0]}
  n50 = Add(A=Add(A=@n16, B=Multiply(A=@n45, B=@n49.MasterBeatMult))) {constB=-1}
  n51 = StaticComponentMaskParameter(None=@n2) {defaultG=true, group=Mask, parameterName=U_Mask}
  n52 = OneMinus(None=@n51)
  n53 = Frac(None=Multiply(A=StaticSwitchParameter(True=@n52, False=@n51) {parameterName=Invert Direction}, B=@n25) {constB=10})
  n54 = OneMinus(None=@n53)
  n55 = Multiply(A=Multiply(A=Multiply(A=@n48) {constB=4}, B=Ceil(None=@n50)), B=Multiply(A=Multiply(A=@n54, B=@n53)) {constB=4})
  n56 = Multiply(
   A=Multiply(
    A=LinearInterpolate(A=@n21, B=@n22, Alpha=@n46) {constA=0.25, constB=5}
    B=ScalarParameter() {defaultValue=1, group=Emissive, parameterName=Chroma Mult}
   )
   B=@n55
  )
  MaterialFunctionCall(Emissive=Multiply(A=Multiply(A=@n20.Result, B=@n56), B=@n55)) {materialFunction=&MF_FX_EmissiveAdjustments}
  n57 = MaterialFunctionCall(
   Fade Length=ScalarParameter() {defaultValue=512, group=Mask, parameterName=CamerFade Length}
   Fade Offset=ScalarParameter() {defaultValue=50, group=Mask, parameterName=CameraFade Offset}
  ) {materialFunction=&CameraDepthFade}
  n58 = Round(None=ComponentMask(None=TextureCoordinate() {vTiling=0.55}) {g=true})
  n59 = TextureCoordinate() {uTiling=0.6, vTiling=0.4}
  n60 = Constant2Vector() {g=0.51, r=0.3}
  n61 = Constant() {r=0.5}
  n62 = Constant() {r=1}
  n63 = SquareRoot(None=ComponentMask(None=TextureCoordinate() {vTiling=0.55}) {r=true})
  n64 = TextureCoordinate() {uTiling=0.6, vTiling=0.4}
  n65 = Constant2Vector() {g=0.25, r=0.5}
  n66 = Constant() {r=0.5}
  n67 = Constant() {r=1}
  n68 = StaticSwitchParameter(
   True=Multiply(A=OneMinus(None=@n58), B=SphereMask(A=@n59, B=@n60, Radius=@n61, Hardness=@n62))
   False=Multiply(A=OneMinus(None=@n63), B=SphereMask(A=@n64, B=@n65, Radius=@n66, Hardness=@n67))
  ) {defaultValue=true, parameterName=Ribbon Mask}
  ShadingPathSwitch(
   Default=DepthFade(Opacity=Multiply(A=@n57.Result, B=@n68), FadeDistance=ScalarParameter() {group=Mask, parameterName=DepthFade})
   Mobile=@n68
  ) {desc=Migrated from FeatureLevelSwitch}
  n69 = DynamicParameter() {defaultValue=<rgba 0 1 0 0>, paramNames [Random, MasterBeatMult, 0, 0]}
  n70 = TextureSample(UVs=Panner(Coordinate=TextureCoordinate() {uTiling=0.4, vTiling=0.35}) {speedX=0.1, speedY=0.2}) {samplerType=SAMPLERTYPE_Normal, texture=&T_Weap_Distort_N_02}
  n71 = Panner(
   Coordinate=Multiply(A=ComponentMask(None=@n70.RGB) {g=true, r=true}) {constB=0.06}
   Speed=ScalarParameter() {defaultValue=-0.2, group=WPO, parameterName=Noise Evolution Speed}
  )
  n72 = Multiply(A=ComponentMask(None=TextureCoordinate()) {g=true}, B=ScalarParameter() {defaultValue=0.4, group=WPO, parameterName=X Frequency Mult})
  n73 = TextureSample(UVs=Add(A=@n69.Random, B=Add(A=@n71, B=@n72))) {samplerType=SAMPLERTYPE_LinearColor, texture=&T_NoisePack_LinearColor}
  n74 = Multiply(A=ComponentMask(None=TextureCoordinate()) {g=true}, B=ScalarParameter() {defaultValue=0.4, group=WPO, parameterName=Y Frequency Mult})
  n75 = TextureSample(UVs=Add(A=@n69.Random, B=Add(A=@n71, B=@n74))) {samplerType=SAMPLERTYPE_LinearColor, texture=&T_NoisePack_LinearColor}
  n76 = Multiply(A=ComponentMask(None=TextureCoordinate()) {g=true}, B=ScalarParameter() {defaultValue=0.4, group=WPO, parameterName=Z Frequency Mult})
  n77 = TextureSample(UVs=Add(A=@n69.Random, B=Add(A=@n71, B=@n76))) {samplerType=SAMPLERTYPE_LinearColor, texture=&T_NoisePack_LinearColor}
  n78 = MaterialFunctionCall(Float3=ConstantBiasScale(None=MaterialFunctionCall(R=@n73.R, G=@n75.G, B=@n77.B) {materialFunction=&AppendMany}.RGB) {bias=-0.5, scale=2}) {materialFunction=&BreakOutFloat3Components}
  n79 = MaterialFunctionCall(
   R=Multiply(A=@n78.R, B=ScalarParameter() {defaultValue=5, group=WPO, parameterName=X_Intesity}) {constB=0.5}
   G=Multiply(A=@n78.G, B=ScalarParameter() {defaultValue=5, group=WPO, parameterName=Y_Intesity}) {constB=5}
   B=Multiply(A=@n78.B, B=ScalarParameter() {defaultValue=5, group=WPO, parameterName=Z_Intensity}) {constB=0.35}
  ) {materialFunction=&AppendMany}
  n80 = AppendVector(A=Constant2Vector(), B=Multiply(A=ScalarParameter() {defaultValue=10, group=WPO, parameterName=Gravity}) {constB=-1})
  n81 = TextureCoordinate()
  n82 = StaticSwitchParameter(True=MaterialFunctionCall(XY=@n81) {materialFunction=&Swizzle}.YX, False=@n81) {group=WPO, parameterName=GRAVITY FADE // Horizontal/Vertical Switch}
  n83 = ComponentMask(None=StaticSwitchParameter(True=OneMinus(None=@n82), False=@n82) {group=WPO, parameterName=GRAVITY FADE // Reverse}) {g=true}
  n84 = TextureCoordinate()
  n85 = Multiply(
   A=Add(A=Transform(None=@n79.RGB) {transformSourceType=TRANSFORMSOURCE_Local}, B=Multiply(A=@n80, B=@n83))
   B=Power(Base=ComponentMask(None=@n84) {g=true}, Exp=ScalarParameter() {defaultValue=1, group=WPO, parameterName=Fade Power})
  )
  n86 = Multiply(A=Multiply(A=@n36, B=Constant() {r=300}), B=ComponentMask(None=TextureCoordinate()) {g=true})
  n87 = Constant()
  n88 = Constant()
  n89 = StaticSwitchParameter(
   True=Transform(None=MaterialFunctionCall(R=@n86, G=@n87, B=@n88) {materialFunction=&AppendMany}.RGB) {transformSourceType=TRANSFORMSOURCE_Local}
   False=Constant()
  ) {parameterName=Enable Rhythm to Scale}
  ShadingPathSwitch(Default=QualitySwitch(Default=Add(A=@n85, B=@n89), Low=@n89), Mobile=@n89) {desc=Migrated from FeatureLevelSwitch}
 ]
}
